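# Material function for a transition effect on the Eddie character body, deforming and masking the mesh.
# A sphere mask on scaled pre-skinned Z, sized by MaterialCurveB ("zipperHeight"), selects the affected region.
# That mask blends thickness-scaled shrinkage, combined with MaterialCurveC inflation, into normal-direction world position offset.
# The inverted, sharpened mask sets opacity mask on incoming material attributes; the bend parameter is declared but unused.
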MaterialFunction /BRCosmetics/Characters/Player/Male/Medium/Bodies/M_MED_Clash_V_Casual/Materials/MF_Transition_Eddie
graph {nodes [
  Constant()
  n1 = GetMaterialAttributes(None=FunctionInput() {inputType=FunctionInput_MaterialAttributes}) {attributeGetTypes [#1]}
  n2 = FunctionInput() {inputName=thickness, inputType=FunctionInput_Scalar}
  n3 = Multiply(A=@n2, B=ScalarParameter() {defaultValue=-3, parameterName=Shrinkage})
  n4 = PreSkinnedPosition()
  n5 = VertexInterpolator(VS=@n4)
  n6 = NamedRerouteDeclaration(None=FunctionInput(Preview=ShaderStageSwitch(PixelShader=@n5.PS, VertexShader=@n4)) {bUsePreviewValueAsDefault=true, inputName=PreskinnedPosition}) {name=PreskinnedPos, nodeColor=<rgba 1 0 0.3281 1>}
  n7 = ScalarParameter() {defaultValue=0.5, desc=zipperHeight, group=Faux_WPO, parameterName=MaterialCurveB}
  n8 = SphereMask(
   A=Divide(A=ComponentMask(None=@n6) {b=true}) {constB=170}
   B=NamedRerouteUsage()
   Radius=NamedRerouteDeclaration(None=LinearInterpolate(Alpha=@n7) {constB=1.5}) {name=zipperHeight, nodeColor=<rgba 0 1 0.6719 1>}
   Hardness=ScalarParameter() {defaultValue=0.92, parameterName=TransitionPower}
  )
  n9 = ScalarParameter() {defaultValue=-0.85, parameterName=Trhinkage}
  n10 = NamedRerouteDeclaration(None=ScalarParameter() {desc=SphereMask Radius, group=Faux_WPO, parameterName=MaterialCurveC}) {name=inflation, nodeColor=<rgba 0 0.6953 1 1>}
  n11 = Multiply(A=VertexNormalWS(), B=Add(A=LinearInterpolate(B=@n3, Alpha=@n8) {constB=-1}, B=LinearInterpolate(B=@n9, Alpha=@n10) {constB=-0.8}))
  n12 = Power(Base=LinearInterpolate(Alpha=@n8) {constA=1, constB=0}) {constExponent=4}
  FunctionOutput(None=SetMaterialAttributes!(MaterialAttributes=@n1.MaterialAttributes, BaseColor=@n1.MaterialAttributes, WorldPositionOffset=@n11, OpacityMask=Saturate(None=@n12)) {attributeSetTypes [#1, #2, #3]})
  MaterialFunctionCall() {materialFunction=&ObjectPivotPoint}
  Multiply(A=@n2, B=@n11)
  NamedRerouteDeclaration(None=ScalarParameter() {desc=drives angle of bend - normalized, group=Faux_WPO, parameterName=MaterialCurveA}) {name=BendAmt, nodeColor=<rgba 0.2656 1 0 1>}
  NamedRerouteUsage()
  ScalarParameter() {parameterName=CenterOfTrans}
 ]}
