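# Master material for the Clash V Casual male character body, built on material attributes as masked, dithered Default Lit for skeletal meshes with morph targets and mesh deformers.
# The shared character PreFX function feeds a ClothFuzz layer, masked by the red channel of the Skin FX texture.
# The isTransition static switch optionally adds MF_Transition_Eddie, driven by pre-skinned position and the texture's green channel as thickness.
# The result passes into the shared character PostFX function.
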Material /BRCosmetics/Characters/Player/Male/Medium/Bodies/M_MED_Clash_V_Casual/Materials/M_Clash_V_Casual_Master
graph {
 properties {
  bUseMaterialAttributes=true
  bUsedWithMeshDeformer=true
  bUsedWithMorphTargets=true
  bUsedWithSkeletalMesh=true
  blendMode=BLEND_Masked
  ditherOpacityMask=true
  pixelDepthOffsetMode=PDOM_Legacy
  usedShadingModels=Default Lit
 }
 nodes [
  n1 = MaterialFunctionCall() {materialFunction=&BaseFN_CharacterFunction_PreFX}
  n2 = TextureSampleParameter2D() {
   parameterName=Skin FX
   samplerType=SAMPLERTYPE_Masks
   texture=/BRCosmetics/Characters/Player/Male/Medium/Bodies/M_MED_Clash_V_Casual/Textures/T_M_MED_Clash_V_Casual_Body_FX
  }
  n3 = MaterialFunctionCall!(Alpha=@n2.R, Gmap=@n1.Gmap, Material=@n1.Gmap) {materialFunction=/Game/Weapons/WeaponSkins/MaterialFunction/ClothFuzz}
  MaterialFunctionCall(
   AO=@n1.AO
   Distance Blend=@n1.AO
   Fresnel=@n1.AO
   Gmap=@n1.AO
   Pre Skinned Local Position=@n1.AO
   Material=StaticSwitchParameter(
    True=MaterialFunctionCall!(PreskinnedPosition=@n1.Pre skinned Local Position, thickness=@n2.G, In=@n3.Result) {materialFunction=&/MF_Transition_Eddie}.Result
    False=@n3.Result
   ) {parameterName=isTransition}
  ) {materialFunction=&BaseFN_CharacterFunction_PostFX}
 ]
}
