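# Physics asset for the left foot mesh of the male medium Clash casual body cosmetic, covering its shoelace bones.
# Small capsule bodies sit on the shoelace base and four lace bones.
# Constraints form two two-segment chains branching from the base, lace 01 to 02 and lace 03 to 04.
# Every joint shares the same angle limits, with twist tighter than swing.
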PhysicsAsset /BRCosmetics/Characters/Player/Male/Medium/Bodies/M_MED_Clash_V_Casual/Meshes/M_MED_Clash_V_Casual_Foot_L_Physics
bodies {
 lace_01_l [{shapeType=Capsule, length=1, radius=1, shapeName=lace_01_l_capsule, center=<-0.8335 0.3345 0.7794>, rotation=<rot 90 0 0>}]
 lace_02_l [{shapeType=Capsule, length=1, radius=1, shapeName=lace_02_l_capsule, center=<-0.8335 0.3345 0.7794>, rotation=<rot 90 0 0>}]
 lace_03_l [{shapeType=Capsule, length=1, radius=1, shapeName=lace_03_l_capsule, center=<-0.469 0.8195 -0.9651>, rotation=<rot 90 0 0>}]
 lace_04_l [{shapeType=Capsule, length=1, radius=1, shapeName=lace_04_l_capsule, center=<-0.7135 0.8195 -0.9651>, rotation=<rot 90 0 0>}]
 shoelace_base_l [{shapeType=Capsule, length=1, radius=1, shapeName=shoelace_base_l_capsule}]
}
constraints [
 {bone1Name=lace_01_l, bone2Name=shoelace_base_l, swing1LimitDegrees=25, swing2LimitDegrees=25, twistLimitDegrees=10}
 {bone1Name=lace_02_l, bone2Name=lace_01_l, swing1LimitDegrees=25, swing2LimitDegrees=25, twistLimitDegrees=10}
 {bone1Name=lace_03_l, bone2Name=shoelace_base_l, swing1LimitDegrees=25, swing2LimitDegrees=25, twistLimitDegrees=10}
 {bone1Name=lace_04_l, bone2Name=lace_03_l, swing1LimitDegrees=25, swing2LimitDegrees=25, twistLimitDegrees=10}
]
