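# Physics asset for the right foot mesh of the casual Clash medium male body cosmetic, simulating the shoelaces.
# Small capsule bodies sit on a shoelace base bone and four lace bones.
# Constraints form two two-segment chains branching from the base, lace 01 to 02 and lace 03 to 04, each joint with the same narrow swing and twist limits.
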PhysicsAsset /BRCosmetics/Characters/Player/Male/Medium/Bodies/M_MED_Clash_V_Casual/Meshes/M_MED_Clash_V_Casual_Foot_R_Physics
bodies {
 lace_01_r [{shapeType=Capsule, length=1, radius=1, shapeName=lace_01_r_capsule, center=<1.4864 -0.0001 0>, rotation=<rot 90 0 0>}]
 lace_02_r [{shapeType=Capsule, length=1, radius=1, shapeName=lace_02_r_capsule, center=<1.4864 -0.0001 0>, rotation=<rot 90 0 0>}]
 lace_03_r [{shapeType=Capsule, length=1, radius=1, shapeName=lace_03_r_capsule, center=<0.8141 -0.0001 0.434>, rotation=<rot 90 0 0>}]
 lace_04_r [{shapeType=Capsule, length=1, radius=1, shapeName=lace_04_r_capsule, center=<0.7912 -0.0001 0.5996>, rotation=<rot 90 0 0>}]
 shoelace_base_r [{shapeType=Capsule, length=1, radius=1, shapeName=shoelace_base_r_capsule}]
}
constraints [
 {bone1Name=lace_04_r, bone2Name=lace_03_r, swing1LimitDegrees=25, swing2LimitDegrees=25, twistLimitDegrees=10}
 {bone1Name=lace_01_r, bone2Name=shoelace_base_r, swing1LimitDegrees=25, swing2LimitDegrees=25, twistLimitDegrees=10}
 {bone1Name=lace_02_r, bone2Name=lace_01_r, swing1LimitDegrees=25, swing2LimitDegrees=25, twistLimitDegrees=10}
 {bone1Name=lace_03_r, bone2Name=shoelace_base_r, swing1LimitDegrees=25, swing2LimitDegrees=25, twistLimitDegrees=10}
]
