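# Post-pose AnimBlueprint for the casual Clash V body that copies the attached parent mesh's pose, then adds secondary and cosmetic corrections.
# Left upperarm yaw relative to the clavicle drives two alphas that blend additive offsets on the cardigan helper bones.
# A visible, tag-matched pickaxe shrinks the right hand to near zero and slims the forearm twist bones.
# Two RigidBody nodes simulate per-foot physics assets in foot bone space.
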AnimBlueprint /BRCosmetics/Characters/Player/Male/Medium/Bodies/M_MED_Clash_V_Casual/Meshes/M_MED_Clash_V_Casual_AnimBP
target_skeleton=&/M_MED_Clash_V_Casual_Skeleton
graphs [
 AnimGraph {nodes [
   "Transform (Modify) Bone - Bone: cardigan_helper_l"(Alpha=1, Translation="0,0,0", Rotation="-12.337,75,-7.254", Scale="1,1,1") {rotationMode=BMM_Additive, alphaBoolBlend {blendOption=Linear}, boneToModify {boneName=cardigan_helper_l}, rotation=<rot -12.337 75 -7.254>}
   n1 = Get aim_alpha() -> (aim_alpha:double)
   n2 = "Transform (Modify) Bone - Bone: cardigan_helper_l"(
    ComponentPose=Local To Component(LocalPose=Copy Pose From Mesh() {bCopyCurves=true, bUseAttachedParent=true})
    Alpha=@n1
    Translation="0,0,0"
    Rotation="-25.239,59.355,0.400"
    Scale="1,1,1"
   ) {rotationMode=BMM_Additive, alphaBoolBlend {blendOption=Linear}, boneToModify {boneName=cardigan_helper_l}, rotation=<rot -25.239 59.355 0.4>}
   n3 = "Transform (Modify) Bone - Bone: cardigan_helper_l"(ComponentPose=@n2, Alpha=Get aim_up_alpha() -> (aim_up_alpha:double), Translation="0,0,0", Rotation="13,15,-7", Scale="1,1,1") {rotationMode=BMM_Additive, alphaBoolBlend {blendOption=Linear}, boneToModify {boneName=cardigan_helper_l}, rotation=<rot 13 15 -7>}
   n4 = "Transform (Modify) Bone - Bone: cardigan_helper_r"(ComponentPose=@n3, Alpha=@n1, Translation="9.316,5.104,0.000", Rotation="0,0,0", Scale="1,1,1") {
    rotationMode=BMM_Additive
    translationMode=BMM_Additive
    alphaBoolBlend {blendOption=Linear}
    boneToModify {boneName=cardigan_helper_r}
    translation=<9.316 5.104>
   }
   n5 = Get Clash Pickaxe() -> (Clash Pickaxe:double)
   n6 = "Transform (Modify) Bone - Bone: hand_r"(ComponentPose=@n4, Alpha=@n5, Translation="0,0,0", Rotation="0,0,0", Scale="0.01,0.01,0.01") {scaleMode=BMM_Replace, scaleSpace=BCS_BoneSpace, alphaBoolBlend {blendOption=Linear}, boneToModify {boneName=hand_r}, scale=<0.01 0.01 0.01>}
   n7 = "Transform (Modify) Bone - Bone: lowerarm_twist_02_r"(ComponentPose=@n6, Alpha=@n5, Translation="0,0,0", Rotation="0,0,0", Scale="0.9,0.9,0.9") {
    scaleMode=BMM_Replace
    scaleSpace=BCS_BoneSpace
    alphaBoolBlend {blendOption=Linear}
    boneToModify {boneName=lowerarm_twist_02_r}
    scale=<0.9 0.9 0.9>
   }
   n8 = "Transform (Modify) Bone - Bone: lowerarm_twist_01_r"(ComponentPose=@n7, Alpha=@n5, Translation="0,0,0", Rotation="0,0,0", Scale="0.9,0.9,0.9") {
    scaleMode=BMM_Replace
    scaleSpace=BCS_BoneSpace
    alphaBoolBlend {blendOption=Linear}
    alphaScaleBiasClamp {bInterpResult=true}
    boneToModify {boneName=lowerarm_twist_01_r}
    scale=<0.9 0.9 0.9>
   }
   n9 = RigidBody(ComponentPose=@n8) {
    bOverrideWorldGravity=true
    overridePhysicsAsset=&/M_MED_Clash_V_Casual_Foot_L_Physics
    simulationSpace=BaseBoneSpace
    alphaBoolBlend {blendOption=Linear}
    baseBoneRef {boneName=foot_l}
    componentAppliedLinearAccClamp=<2000 2000 2000>
    componentLinearAccScale=<0.3 0.3 0.35>
    componentLinearVelScale=<0.3 0.3 0.35>
    overrideWorldGravity {z=-980}
   }
   n10 = RigidBody(ComponentPose=@n9) {
    bOverrideWorldGravity=true
    overridePhysicsAsset=&/M_MED_Clash_V_Casual_Foot_R_Physics
    simulationSpace=BaseBoneSpace
    alphaBoolBlend {blendOption=Linear}
    baseBoneRef {boneName=foot_r}
    componentAppliedLinearAccClamp=<2000 2000 2000>
    componentLinearAccScale=<0.3 0.3 0.35>
    componentLinearVelScale=<0.3 0.3 0.35>
    overrideWorldGravity {z=-980}
   }
   Output Pose(Result=Component To Local(ComponentPose=@n10)) {layerGroup=None}
  ]}
 EventGraph {nodes [
   Event BlueprintUpdateAnimation()
   n1 = GetOwnerSkeletalMeshComponent()
   n2 = BreakRotator(InRot=BreakTransform(InTransform=GetDeltaTransformFromRefPose(self=@n1, BoneName=upperarm_l, BaseName=clavicle_l)).Rotation)
   > Set aim_alpha(aim_alpha:double=MapRangeClamped(Value=@n2.Yaw, InRangeA=5, InRangeB=60)) -> (Output_Get:double)
   > Set aim_up_alpha(aim_up_alpha:double=MapRangeClamped(Value=@n2.Yaw, InRangeA=60, InRangeB=90)) -> (Output_Get:double)
   > Set Fort Pawn(Fort Pawn:Fort Pawn Object Reference=TryGetFortPawn()) -> (Output_Get:Fort Pawn Object Reference)
   n3 = Get Fort Pawn() -> (Fort Pawn:Fort Pawn Object Reference)
   > n4 = Is Valid(InputObject=@n3)
   n5 = Set Weapon Visible(execute=@n4.Is Not Valid, Weapon Visible:bool=false) -> (Output_Get:bool)
   n6 = Get Weapon Visible() -> (Weapon Visible:bool)
   n7 = Branch(execute=[@n11.then, @n5.then], Condition=@n6)
   n8 = PickaxeItemDefinitionHasAnyMetaTag(
    Tags=Make Array!(
     "[0]"=(TagName="Cosmetics.Pickaxe.Vertigo")
     "[1]":Gameplay Tag Structure=(TagName="Cosmetics.Exclusives.ClashV")
     "[2]":Gameplay Tag Structure
    ) -> (Array)
    bRequiresEquipped=true
   )
   > Set Clash Pickaxe(Clash Pickaxe:double="To Float (Boolean)"(InBool=AND Boolean(A=@n6, B=@n8))) -> (Output_Get:double)
   Set Clash Pickaxe(execute=@n7.else, Clash Pickaxe=0)
   n9 = IsWeaponHolstered!(self:Fort Pawn Object Reference=@n3) -> (ReturnValue:bool)
   n10 = Get CurrentWeapon!(self:Fort Pawn Object Reference=@n3) -> (CurrentWeapon:Fort Weapon Object Reference)
   n11 = Set Weapon Visible(execute=@n4.Is Valid, Weapon Visible=AND Boolean(A=IsValid(Object=@n10), B=NOT Boolean(A=@n9)))
   WeaponDataHasMatchingGameplayTag!(
    self:Fort Weapon Object Reference=Get CurrentWeapon(self=Get Fort Pawn())
    TagToCheck:Gameplay Tag Structure=(TagName="Cosmetics.Exclusives.ClashV")
   ) -> (ReturnValue:bool)
  ]}
]
variables [
 aim_alpha {type=number}
 aim_up_alpha {type=number}
 Clash Pickaxe {type=number}
 Fort Pawn {type=FortPawn, default=null}
 Weapon Visible {type=boolean}
]
